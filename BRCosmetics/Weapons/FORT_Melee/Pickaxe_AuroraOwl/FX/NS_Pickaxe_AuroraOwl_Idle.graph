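# Idle Niagara effect for the AuroraOwl pickaxe that sprays fire-extinguisher vapor from a socket.
# ExtinguisherTip draws socket-following mesh and streak sprites, while Nozzle and World emitters add animated smoke puffs, with World adding movement-based spawning and gravity.
# Normalized reactivity time gates spawning and scales size, alpha and speed, with front-end modifiers and time-of-day tinting.
# Puffs age faster with velocity, while a separate AgeAlt clock drives their SubUV animation.
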NiagaraSystem /BRCosmetics/Weapons/FORT_Melee/Pickaxe_AuroraOwl/FX/NS_Pickaxe_AuroraOwl_Idle
dependencies {
 usedDataInterfaces [
  &NiagaraDataInterfaceCurve
  &NiagaraDataInterfaceSpriteRendererInfo
  &NiagaraDataInterfaceVectorCurve
  &NiagaraDataInterfaceVector2DCurve
  &NiagaraDataInterfaceSocketReader
  &NiagaraDataInterfaceColorCurve
 ]
 usedDynamicInputs [
  &Vector2DFromFloat
  &FNTimeOfDayScalar
  &ColorFromCurve
 ]
 usedModules [
  &self:SystemSpawnScript.SystemScriptSource.SystemScriptGraph.NiagaraNodeAssignment_1.SetVariables_#1
  &self:SystemSpawnScript.SystemScriptSource.SystemScriptGraph.NiagaraNodeAssignment_3.SetVariables_#2
  &SystemState
  &self:SystemSpawnScript.SystemScriptSource.SystemScriptGraph.NiagaraNodeAssignment_0.SetVariables_#3
  &self:SystemSpawnScript.SystemScriptSource.SystemScriptGraph.NiagaraNodeAssignment_7.SetVariables_#4
  &self:SystemSpawnScript.SystemScriptSource.SystemScriptGraph.NiagaraNodeAssignment_2.SetVariables_#5
 ]
}
emitters {
 ExtinguisherTip {
  emitterUpdateScript [
   &EmitterState(
    Inactive Response=::NewEnumerator1
    Life Cycle Mode=::NewEnumerator1
    Loop Behavior=::NewEnumerator1
    Loop Delay=0.0333
    Loop Duration=1.5
    UseLoopDelay=true
   )
   &SpawnBurst_Instantaneous(
    Spawn Probability=&MakeCustomFloatFromBool(Bool=&SetBoolByFloatComparison(A=$System.NormalizedReactivityTime, B=0.01), True Float=1, False Float=0)
    Use Spawn Probability=true
    UseLoopCountLimit=false
   )
   &SpawnRate(
    Spawn Probability=&FloatFromCurve(
     FloatCurve=DataInterface() {propertyValues {Curve {keys [<at 0 0 RCIM_Constant>,<at 0.4 1 RCIM_Constant>]}}}
     CurveIndex=$Emitter.Age
     Scale Curve=&FloatFromCurve(
      FloatCurve=DataInterface() {propertyValues {Curve {keys [
          {arriveTangent=0.2704, interpMode=RCIM_Cubic, leaveTangent=0.2704, tangentMode=RCTM_User, time=0, value=0}
          {arriveTangent=2.7605, arriveTangentWeight=0.2, interpMode=RCIM_Cubic, leaveTangent=2.7605, tangentMode=RCTM_User, time=1, value=1}
         ]}}}
      CurveIndex=$System.NormalizedReactivityTime
     )
    )
    SpawnGroup=1
    SpawnRate=35
    Use Spawn Probability=true
   )
  ]
  particleSpawnScript [
   &InitializeParticle(
    Color=&MergeHelper_ColorSwitch(
     0=<rgba 0.975 0.9937 1 5>
     1=<rgba 0.925 0.9812 1 0.4>
     2=&RandomRangeLinearColor(Minimum=<rgba 0.925 0.9875 1 0.15>, Maximum=<rgba 0.95 0.9792 1 0.3>)
     3=?
     Index=$Emitter.SpawnGroup
     Num=::NewEnumerator0
    )
    Lifetime=&MergeHelper_FloatSwitch(
     0=5
     1=&RandomRangeFloat(Minimum=0.075, Maximum=0.125)
     2=&RandomRangeFloat(Minimum=1.15, Maximum=1.4)
     3=?
     Index=$Emitter.SpawnGroup
     Num=::NewEnumerator0
    )
    Mesh Scale=<0.065 0.2 0.065>
    Mesh Scale Mode=::NewEnumerator4
    Sprite Size=&RandomRangeVector2D(Minimum=<0.7 10>, Maximum=<1 20>)
    Sprite Size Mode=::NewEnumerator4
    UsePositionOffset=false
    Write Material Random=true
   )
   &SocketLocation(
    SamplePosOffset=true
    Sampled Position Offset=<-2 0 0>
    Socket Reader=$System.Socket Reader
    Socket Sampling Mode=::NewEnumerator3
    VelocitySpeedLimit=false
   )
   &AddVelocity(
    Additional Quaternion Rotation=false
    Cone Angle=15
    Module.UseConeFalloff=false
    Module.UseVelDistribution=false
    Override Seed=false
    Rotation Mode=::NewEnumerator3
    Rotation Quaternion:Quat4f=$Output.SocketLocation.SampledRotation
    Velocity Mode=::NewEnumerator2
    Velocity Speed=&MakeFloatFromVector(
     Vector=&MergeHelper_VectorSwitch(
      0=<0 0 0>
      1=&MakeVector(
       X=&FN_FrontEndFloatModifier(Gameplay=&RandomRangeFloat(Minimum=300, Maximum=550), Front End=$System.FrontEndVelocity)
       Y=0
       Z=0
      )
      2=&MakeVector(X=&RandomRangeFloat(Minimum=300, Maximum=550), Y=0, Z=0)
      3=?
      Index=$Emitter.SpawnGroup
      Num=::NewEnumerator0
     )
     Channel=::NewEnumerator0
    )
   )
   /SetVariables(Particles.VisibilityTag=$Emitter.SpawnGroup) {isSetParameters=true}
  ]
  particleUpdateScript [
   &ParticleState()
   &SocketLocation(
    SamplePosOffset=true
    Sampled Position Offset=<-2 0 0>
    Socket Reader=$System.Socket Reader
    Socket Sampling Mode=::NewEnumerator3
    VelocitySpeedLimit=false
   ) {module=SocketLocation001}
   &Drag(Drag=&RandomRangeFloat(Minimum=5, Maximum=8), Use Linear Drag=true, Use Rotational Drag=true)
   &ScaleMeshSize(Scale Factor=&VectorFromCurve(
     VectorCurve=DataInterface() {propertyValues {
       XCurve {keys [<at 0 0 RCIM_Cubic>,<at 0.1 1 RCIM_Cubic>]}
       YCurve {defaultValue=0, keys [<at 0 0 RCIM_Cubic>,<at 0.15 1 RCIM_Cubic>]}
       ZCurve {keys [<at 0 0 RCIM_Cubic>,<at 0.1 1 RCIM_Cubic>]}
      }}
     CurveIndex=$Particles.Age
    ))
   &ScaleSpriteSize(
    Scale Factor=&Vector2DFromCurve(
     Vector2Curve=DataInterface() {propertyValues {
       XCurve {defaultValue=1, keys [<at 0 0.5 RCIM_Cubic>,<at 1 0 RCIM_Cubic>]}
       YCurve {defaultValue=1, keys [<at 0 0.5 RCIM_Cubic>,<at 1 1.5 RCIM_Cubic>]}
      }}
     ScaleCurve=<1 1>
    )
    Scale Sprite Size Mode=::NewEnumerator2
   )
   &ScaleColor(
    Scale Alpha=&MergeHelper_FloatSwitch(
     0=&FloatFromCurve(
      FloatCurve=DataInterface() {propertyValues {Curve {keys [{arriveTangent=1.9904, interpMode=RCIM_Cubic, leaveTangent=1.9904, tangentMode=RCTM_User, time=0, value=0}, <at 1 1 RCIM_Cubic>]}}}
      CurveIndex=$System.NormalizedReactivityTime
      Scale Curve=&FloatFromCurve(
       FloatCurve=DataInterface() {propertyValues {Curve {keys [<at 0.5 1 RCIM_Cubic>, {arriveTangent=-3.1121, interpMode=RCIM_Cubic, leaveTangent=-3.1121, tangentMode=RCTM_User, time=1.5, value=0}]}}}
       CurveIndex=$Particles.Age
      )
     )
     1=&FloatFromCurve(FloatCurve=DataInterface() {propertyValues {Curve {keys [<at 0 1 RCIM_Cubic>,<at 1 0 RCIM_Cubic>]}}})
     2=&FloatFromCurve(FloatCurve=DataInterface() {propertyValues {Curve {keys [
          {interpMode=RCIM_Cubic, tangentMode=RCTM_User, time=0, value=0}
          {interpMode=RCIM_Cubic, tangentMode=RCTM_User, time=0.1, value=1}
          {tangentMode=RCTM_User, time=1, value=0}
         ]}}})
     3=?
     Index=$Particles.VisibilityTag
     Num=::NewEnumerator0
    )
    Scale RGB=&MakeVectorFromLinearColorRGB(Color=$System.TODColor)
    ScaleA=true
    ScaleRGB=true
   )
   &SolveForcesAndVelocity(Clamp Velocity=false, Limit Acceleration=false)
  ]
  rendererClasses [&NiagaraMeshRendererProperties, &NiagaraSpriteRendererProperties]
  renderers [
   {rendererClass=&NiagaraMeshRendererProperties, rendererIndex=0}
   {rendererClass=&NiagaraSpriteRendererProperties, rendererIndex=1}
  ]
 }
 FireExtinguisher_Nozzle {
  emitterUpdateScript [
   &EmitterState(Life Cycle Mode=::NewEnumerator1, Loop Behavior=::NewEnumerator1, Loop Duration=1.4, UseLoopDelay=true)
   &SpawnBurst_Instantaneous(
    Spawn Count=5
    Spawn Probability=&MakeCustomFloatFromBool(Bool=&SetBoolByFloatComparison(A=$System.NormalizedReactivityTime, B=0.01), True Float=1, False Float=0)
    Use Spawn Probability=true
    UseLoopCountLimit=false
   )
   &SpawnRate(
    Spawn Probability=&MakeCustomFloatFromBool(Bool=&SetBoolByFloatComparison(A=$System.NormalizedReactivityTime, B=0.01), True Float=1, False Float=0)
    SpawnRate=15
    Use Spawn Probability=true
   )
  ]
  particleSpawnScript [
   &InitializeParticle(
    Color=&RandomRangeLinearColor(Minimum=<rgba 0.925 0.9875 1 0.2>, Maximum=<rgba 0.95 0.9792 1 0.35>)
    Lifetime=&RandomRangeFloat(Minimum=0.8, Maximum=1)
    Position Offset=<10 0 0>
    Sprite Rotation Mode=::NewEnumerator2
    Sprite Size Mode=::NewEnumerator3
    Sprite UV Mode=::NewEnumerator3
    Uniform Sprite Size=&Multiply_Float(A=&RandomRangeFloat(Minimum=21, Maximum=27), B=$System.FrontEndScale)
    UsePositionOffset=true
    Write Material Random=true
   )
   &CameraOffset(Camera Offset Amount=&RandomRangeFloat(Minimum=20, Maximum=25))
   &SocketLocation(
    "Inherit Velocity (Scale)"=-1
    Interpolate Socket Sampling=true
    SamplePosOffset=false
    Sampled Velocity Coordinate Space:ENiagaraCoordinateSpace=::Local
    Socket Reader=$System.Socket Reader
    Socket Sampling Mode=::NewEnumerator3
    VelocitySpeedLimit=false
   )
   &AddVelocity(
    Additional Quaternion Rotation=false
    Cone Angle=15
    Module.UseConeFalloff=false
    Module.UseVelDistribution=false
    Override Seed=false
    Rotation Mode=::NewEnumerator3
    Rotation Quaternion:Quat4f=$Output.SocketLocation.SampledRotation
    Velocity Mode=::NewEnumerator2
    Velocity Speed=&MakeFloatFromVector(
     Vector=&MakeVector(
      X=&Multiply_Float(A=&RandomRangeFloat(Minimum=150, Maximum=200), B=$System.FrontEndVelocity)
      Y=0
      Z=0
     )
     Channel=::NewEnumerator0
    )
   )
   /SetVariables() {isSetParameters=true}
  ]
  particleUpdateScript [
   &ParticleState()
   /SetVariables(Particles.VelocityIntensity:NiagaraFloat=&Divide_Float(A=&VectorLength(Vector=$Engine.Owner.Velocity), B=75)) {isSetParameters=true}
   /SetVariables(Particles.Age:NiagaraFloat=&Add_Float(
     A=$Particles.Age
     B=&Multiply_Float(A=$Particles.VelocityIntensity, B=&Multiply_Float(A=$Particles.Lifetime, B=$Engine.DeltaTime))
    )) {isSetParameters=true}
   /SetVariables(Particles.AgeAlt:NiagaraFloat=&Add_Float(A=$Particles.AgeAlt, B=$Engine.DeltaTime)) {isSetParameters=true}
   &Drag(
    Drag=&FloatFromCurve(
     FloatCurve=DataInterface() {propertyValues {Curve {keys [
         {interpMode=RCIM_Cubic, tangentMode=RCTM_User, time=0, value=0}
         {arriveTangent=22.0021, interpMode=RCIM_Cubic, leaveTangent=22.0021, tangentMode=RCTM_User, time=0.06, value=1}
        ]}}}
     Scale Curve=8
    )
    Use Linear Drag=true
    Use Rotational Drag=true
   )
   &SubUVAnimation(
    "Sprite Renderer "=DataInterface() {propertyValues {SpriteRenderer=&self:FireExtinguisher_Nozzle_0.NiagaraSpriteRendererProperties_0}}
    SubUV Lookup Index:NiagaraFloat=&Divide_Float(A=$Particles.AgeAlt, B=$Particles.Lifetime)
    UseEndFrame=true
    UseStartFrame=true
   )
   &SpriteRotationRate(Rotation Rate=&RandomRangeFloat(Minimum=5, Maximum=50))
   &DynamicMaterialParameters(
    Index 0 Param 1=&FloatFromCurve(FloatCurve=DataInterface() {propertyValues {Curve {keys [<at 0 1>,<at 1 -0.5>]}}})
    Index 0 Param 2=1
    Index 0 Param 3=1
    Index 0 Param 4=0.5
    Module.Index0Float0Write=true
    Module.Index0Float1Write=true
    Module.Index0Float2Write=false
    Module.Index0Float3Write=true
   )
   &SolveForcesAndVelocity(Clamp Velocity=false, Limit Acceleration=false)
   &ScaleSpriteSize(
    Scale Sprite Size Mode=::NewEnumerator0
    Uniform Scale Factor=&Multiply_Float(
     A=&Lerp_Float(A=0.2, B=&Lerp_Float(A=0.5, B=1.3821, Alpha=$System.NormalizedReactivityTime), Alpha=$Particles.NormalizedAge)
     B=1
    )
   )
   &ScaleColor(
    Scale Alpha=&Multiply_Float(
     A=&FloatFromCurve(FloatCurve=DataInterface() {propertyValues {Curve {keys [
          {interpMode=RCIM_Cubic, tangentMode=RCTM_User, time=0, value=0}
          {interpMode=RCIM_Cubic, tangentMode=RCTM_User, time=0.3, value=1}
          {tangentMode=RCTM_User, time=1, value=0}
         ]}}})
     B=&FloatFromCurve(
      FloatCurve=DataInterface() {propertyValues {Curve {keys [{arriveTangent=2.2094, time=0, value=0}, {arriveTangent=0.0224, leaveTangent=0.0224, time=1, value=1}]}}}
      CurveIndex=$Emitter.Age
     )
    )
    Scale RGB=&MakeVectorFromLinearColorRGB(Color=$System.TODColor)
    ScaleA=true
    ScaleRGB=true
   )
  ]
  rendererClasses [&NiagaraSpriteRendererProperties]
  renderers [{rendererClass=&NiagaraSpriteRendererProperties}]
 }
 FireExtinguisher_World {
  emitterUpdateScript [
   &EmitterState(Life Cycle Mode=::NewEnumerator1, Loop Behavior=::NewEnumerator1, Loop Delay=0.1, Loop Duration=1.3, UseLoopDelay=true)
   &SpawnBurst_Instantaneous(
    Spawn Count=5
    Spawn Probability=&MakeCustomFloatFromBool(Bool=&SetBoolByFloatComparison(A=$System.NormalizedReactivityTime, B=0.01), True Float=1, False Float=0)
    Use Spawn Probability=true
    UseLoopCountLimit=false
   )
   &SpawnPerUnit(
    Max Movement Threshold=400
    Module.Use Max Movement Threshold=false
    Module.Use Movement Tolerance=false
    Spawn Probability:NiagaraFloat=&MakeCustomFloatFromBool(Bool=&SetBoolByFloatComparison(A=$System.NormalizedReactivityTime, B=0.01), True Float=1, False Float=0)
    Spawn Spacing=7
    Use Spawn Probability=true
   )
   &SpawnRate(
    Spawn Probability=&MakeCustomFloatFromBool(Bool=&SetBoolByFloatComparison(A=$System.NormalizedReactivityTime, B=0.01), True Float=1, False Float=0)
    SpawnRate=35
    Use Spawn Probability=true
   )
  ]
  particleSpawnScript [
   &InitializeParticle(
    Color=&RandomRangeLinearColor(Minimum=<rgba 0.925 0.9875 1 0.15>, Maximum=<rgba 0.95 0.9792 1 0.3>)
    Lifetime=&RandomRangeFloat(Minimum=1.15, Maximum=1.4)
    Position Offset=<10 0 0>
    Sprite Rotation Mode=::NewEnumerator2
    Sprite Size Mode=::NewEnumerator3
    Sprite UV Mode=::NewEnumerator3
    Uniform Sprite Size=&RandomRangeFloat(Minimum=55, Maximum=70)
    UsePositionOffset=true
    Write Material Random=true
   )
   &SocketLocation(
    Interpolate Socket Sampling=true
    SamplePosOffset=false
    Socket Reader=$System.Socket Reader
    Socket Sampling Mode=::NewEnumerator3
    VelocitySpeedLimit=false
   )
   &ShapeLocation(MaskSpawnGroup=false, Override Seed=false, Sphere Radius=3, UseOwnerScale=true)
   &InheritVelocity(
    Inherited Velocity Amount Scale=&FN_FrontEndVector3Modifier(Gameplay=<1 1 1>, Front End=<0 0 0>)
    Inherited Velocity Speed Limit=650
    Speed Limit=true
    UseSpeedThreshold=false
   )
   &AddVelocity(
    Additional Quaternion Rotation=false
    Cone Angle=35
    Module.UseConeFalloff=false
    Module.UseVelDistribution=false
    Override Seed=false
    Rotation Mode=::NewEnumerator3
    Rotation Quaternion:Quat4f=$Output.SocketLocation.SampledRotation
    Velocity Mode=::NewEnumerator2
    Velocity Speed=&Multiply_Float(
     A=&Multiply_Float(
      A=&RandomRangeFloat(Minimum=388, Maximum=400)
      B=&Lerp_Float(A=0.35, B=1, Alpha=$System.NormalizedReactivityTime)
     )
     B=$System.FrontEndVelocity
    )
   )
   &CameraOffset(Camera Offset Amount=10)
   /SetVariables() {isSetParameters=true}
  ]
  particleUpdateScript [
   &ParticleState()
   /SetVariables(Particles.VelocityIntensity:NiagaraFloat=&Divide_Float(A=&VectorLength(Vector=$Particles.Velocity), B=225)) {isSetParameters=true}
   /SetVariables(Particles.Age:NiagaraFloat=&Add_Float(
     A=$Particles.Age
     B=&Multiply_Float(A=$Particles.VelocityIntensity, B=&Multiply_Float(A=$Particles.Lifetime, B=$Engine.DeltaTime))
    )) {isSetParameters=true}
   /SetVariables(Particles.AgeAlt:NiagaraFloat=&Add_Float(A=$Particles.AgeAlt, B=$Engine.DeltaTime)) {isSetParameters=true}
   &Drag(
    Drag=&FloatFromCurve(
     FloatCurve=DataInterface() {propertyValues {Curve {keys [
         {interpMode=RCIM_Cubic, tangentMode=RCTM_User, time=0, value=0}
         {arriveTangent=22.0021, interpMode=RCIM_Cubic, leaveTangent=22.0021, tangentMode=RCTM_User, time=0.06, value=1}
        ]}}}
     Scale Curve=8
    )
    Use Linear Drag=true
    Use Rotational Drag=true
   )
   &SubUVAnimation(
    "Sprite Renderer "=DataInterface() {propertyValues {SpriteRenderer=&self:FireExtinguisher_World_0.NiagaraSpriteRendererProperties_0}}
    SubUV Lookup Index:NiagaraFloat=&Divide_Float(A=$Particles.AgeAlt, B=$Particles.Lifetime)
    UseEndFrame=true
    UseStartFrame=true
   )
   &SpriteRotationRate(Rotation Rate=&RandomRangeFloat(Minimum=5, Maximum=50))
   &DynamicMaterialParameters(
    Index 0 Param 1=&FloatFromCurve(FloatCurve=DataInterface() {propertyValues {Curve {keys [<at 0 1>,<at 1 -0.5>]}}})
    Index 0 Param 2=1
    Index 0 Param 3=1
    Index 0 Param 4=0.5
    Module.Index0Float0Write=true
    Module.Index0Float1Write=true
    Module.Index0Float2Write=false
    Module.Index0Float3Write=true
   )
   &GravityForce(Gravity=<0 0 -200>)
   &SolveForcesAndVelocity(Clamp Velocity=false, Limit Acceleration=false)
   &ScaleSpriteSize(
    Scale Sprite Size Mode=::NewEnumerator0
    Uniform Scale Factor=&Multiply_Float(
     A=&Lerp_Float(A=0.2, B=&Lerp_Float(A=0.5, B=1.3821, Alpha=$System.NormalizedReactivityTime), Alpha=$Particles.NormalizedAge)
     B=1
    )
   )
   &ScaleColor(
    Scale Alpha=&Multiply_Float(
     A=&FloatFromCurve(FloatCurve=DataInterface() {propertyValues {Curve {keys [
          {interpMode=RCIM_Cubic, tangentMode=RCTM_User, time=0, value=0}
          {interpMode=RCIM_Cubic, tangentMode=RCTM_User, time=0.1, value=1}
          {tangentMode=RCTM_User, time=1, value=0}
         ]}}})
     B=&FloatFromCurve(
      FloatCurve=DataInterface() {propertyValues {Curve {keys [{arriveTangent=2.2094, time=0, value=0}, {arriveTangent=0.0224, leaveTangent=0.0224, time=1, value=1}]}}}
      CurveIndex=$Emitter.Age
     )
    )
    Scale RGB=&MakeVectorFromLinearColorRGB(Color=$System.TODColor)
    ScaleA=true
    ScaleRGB=true
   )
  ]
  rendererClasses [&NiagaraSpriteRendererProperties]
  renderers [{rendererClass=&NiagaraSpriteRendererProperties}]
 }
}
summary {
 systemName=NS_Pickaxe_AuroraOwl_Idle
 emitters [
  {emitterName=FireExtinguisher_World, rendererClasses [&NiagaraSpriteRendererProperties]}
  {
   emitterName=ExtinguisherTip
   rendererClasses [&NiagaraMeshRendererProperties, &NiagaraSpriteRendererProperties]
  }
  {emitterName=FireExtinguisher_Nozzle, rendererClasses [&NiagaraSpriteRendererProperties]}
 ]
 userVariables [
  User.bIsFrontEnd {type=NiagaraBool}
  User.bIsFrontEndPreview {type=NiagaraBool}
  User.DebugReactivity {type=NiagaraBool}
  User.ReactivityTime {type=NiagaraFloat, defaultValue=0, description=""}
 ]
}
