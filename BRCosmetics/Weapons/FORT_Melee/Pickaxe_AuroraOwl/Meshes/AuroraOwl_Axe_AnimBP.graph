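# Animation Blueprint for the AuroraOwl pickaxe that procedurally animates its swinging links and mechanical parts.
# Two spring-driven, angle-limited Anim Dynamics nodes simulate Link_01 then Link_03 on the input pose.
# A Look At chain then aims paired Pistol bones at opposing targets, plus Link_02 and Rod_01, like piston linkages.
# An extra Link_01 dynamics node is disconnected, and the EventGraph holds no logic.
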AnimBlueprint /BRCosmetics/Weapons/FORT_Melee/Pickaxe_AuroraOwl/Meshes/AuroraOwl_Axe_AnimBP
target_skeleton=&/AuroraOwl_Axe_Skeleton
graphs [
 AnimGraph {nodes [
   Anim Dynamics - Bone: Link_01!(ComponentPose, bAlphaBoolEnabled:bool=True, ExternalForce="0,0,0") -> (Pose) {
    alphaInputType=Bool
    angularDampingOverride=0.99
    bOverrideAngularDamping=true
    bUseGravityOverride=true
    simulationSpace=BoneRelative
    alphaBoolBlend {blendOption=Linear}
    boundBone {boneName=Link_01}
    componentLinearAccScale {y=0.5}
    componentLinearVelScale {y=0.25}
    gravityOverride {z=-980}
    physicsBodyDefinitions [{
      boundBone {boneName=Link_01}
      boxExtents=<20 20 10>
      constraintSetup {angularLimitsMax=<0 0 15>, angularLimitsMin=<0 0 -15>, angularTarget=<0 0 0>}
      localJointOffset=<0 0 -7>
     }]
    relativeSpaceBone {boneName=b_AR_Root}
   }
   AnimGraphNode_SubInput_0 = Input Pose()
   n1 = Anim Dynamics - Bone: Link_01!(ComponentPose=Local To Component(LocalPose=@AnimGraphNode_SubInput_0), Alpha:float=1, ExternalForce="0,0,0") -> (Pose) {
    angularDampingOverride=1
    angularSpringConstant=14000
    bAngularSpring=true
    bGravityOverrideInSimSpace=true
    bOverrideAngularDamping=true
    bUseGravityOverride=true
    boundBone {boneName=Link_01}
    componentAppliedLinearAccClamp=<3000 3000 2500>
    componentLinearAccScale=<0.3 0.3 0.3>
    componentLinearVelScale=<0.3 0.3 0.3>
    gravityOverride {z=-980}
    physicsBodyDefinitions [{
      boundBone {boneName=Link_01}
      boxExtents=<30 30 30>
      constraintSetup {angularLimitsMax=<0 0 20>, angularLimitsMin=<0 0 -10>, angularTarget=<1 0 0>}
      localJointOffset=<1 0 0>
     }]
    simSpaceSettings {maxAngularAcceleration=35, maxAngularVelocity=7, simSpaceAngularAlpha=0.8, externalAngularVelocity=<3 7 3>}
   }
   n2 = Anim Dynamics - Bone: Link_03(ComponentPose=@n1, Alpha=1, ExternalForce="0,0,0") {
    angularDampingOverride=1
    angularSpringConstant=12000
    bAngularSpring=true
    bGravityOverrideInSimSpace=true
    bOverrideAngularDamping=true
    bUseGravityOverride=true
    boundBone {boneName=Link_03}
    componentAppliedLinearAccClamp=<3000 3000 2500>
    componentLinearAccScale=<0.1 0.1 0.1>
    componentLinearVelScale=<0.1 0.1 0.1>
    gravityOverride {z=-980}
    physicsBodyDefinitions [{
      boundBone {boneName=Link_03}
      boxExtents=<35 35 35>
      constraintSetup {angularLimitsMax=<0 0 35>, angularLimitsMin=<0 0 -30>, angularTarget=<1 0 0>}
      localJointOffset=<0.1 0 0>
     }]
    simSpaceSettings {maxAngularAcceleration=40, maxAngularVelocity=8, simSpaceAngularAlpha=0.6, externalAngularVelocity=<10 10 10>}
   }
   n3 = Look At - Bone: Pistol_b_1(ComponentPose=@n2, Alpha=1) {
    alphaBoolBlend {blendOption=Linear}
    boneToModify {boneName=Pistol_b_1}
    lookAtLocation {x=0}
    lookAtTarget {boneReference {boneName=Pistol_t_1_main}}
    lookAt_Axis {axis=<1 0>}
    lookUp_Axis {axis {x=0, z=1}}
   }
   n4 = Look At - Bone: Pistol_b_2(ComponentPose=@n3, Alpha=1) {
    alphaBoolBlend {blendOption=Linear}
    boneToModify {boneName=Pistol_b_2}
    lookAtLocation {x=0}
    lookAtTarget {boneReference {boneName=Pistol_t_2_main}}
    lookAt_Axis {axis=<1 0>}
    lookUp_Axis {axis {x=0, z=1}}
   }
   n5 = Look At - Bone: Pistol_t_1(ComponentPose=@n4, Alpha=1) {
    alphaBoolBlend {blendOption=Linear}
    boneToModify {boneName=Pistol_t_1}
    lookAtLocation {x=0}
    lookAtTarget {boneReference {boneName=Pistol_b_1_main}}
    lookAt_Axis {axis=<1 0>}
    lookUp_Axis {axis {x=0, z=1}}
   }
   n6 = Look At - Bone: Pistol_t_2(ComponentPose=@n5, Alpha=1) {
    alphaBoolBlend {blendOption=Linear}
    boneToModify {boneName=Pistol_t_2}
    lookAtLocation {x=0}
    lookAtTarget {boneReference {boneName=Pistol_b_2_main}}
    lookAt_Axis {axis=<1 0>}
    lookUp_Axis {axis {x=0, z=1}}
   }
   n7 = Look At - Bone: Link_02(ComponentPose=@n6, Alpha=1) {
    alphaBoolBlend {blendOption=Linear}
    boneToModify {boneName=Link_02}
    lookAtLocation=<0 0>
    lookAtTarget {boneReference {boneName=Rod_02}}
    lookAt_Axis {axis=<1 0>}
    lookUp_Axis {axis {x=0, z=1}}
   }
   n8 = Look At - Bone: Rod_01(ComponentPose=@n7, Alpha=1) {
    alphaBoolBlend {blendOption=Linear}
    boneToModify {boneName=Rod_01}
    lookAtLocation {x=0}
    lookAtTarget {boneReference {boneName=Rod_aim}}
    lookAt_Axis {axis=<1 0>}
    lookUp_Axis {axis {x=0, z=1}}
   }
   Output Pose(Result=Component To Local(ComponentPose=@n8)) {layerGroup=None}
  ]}
 EventGraph {nodes [
   TryGetPawnOwner()
   Event BlueprintUpdateAnimation()
  ]}
]
variables []
